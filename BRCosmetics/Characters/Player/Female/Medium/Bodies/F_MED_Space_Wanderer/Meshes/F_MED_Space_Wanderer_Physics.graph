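# Physics asset for the F_MED_Space_Wanderer female player cosmetic body, set up for its dynamic suspender bones.
# Each side chains box bodies from a tiny base through parent and four segments, linked by swing- and twist-limited constraints.
# Capsules on pelvis, thighs, calves and feet have no constraints, apparently serving as colliders for the straps.
# Limits are asymmetric: the right chain is tighter at the base and looser at the lower segments.
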PhysicsAsset /BRCosmetics/Characters/Player/Female/Medium/Bodies/F_MED_Space_Wanderer/Meshes/F_MED_Space_Wanderer_Physics
bodies {
 calf_l [
  {shapeType=Capsule, length=34.625, radius=6.125, center=<-5.0807 -2.3121 -0.2068>, rotation=<rot 82.4866 -174.1614 180>}
  {shapeType=Capsule, length=28.125, radius=8, center=<-28.263 -2.9403 -0.4645>, rotation=<rot 87.5156 -177.4455 -172.6849>}
 ]
 calf_r [
  {shapeType=Capsule, length=35.25, radius=5.5, center=<5.6219 1.9387 -0.1147>, rotation=<rot 81.6244 -174.1614 180>}
  {shapeType=Capsule, length=28.125, radius=8, center=<29.6915 3.4262 1.7768>, rotation=<rot 86.1903 179.9999 -175.2359>}
 ]
 dyn_suspender_l_01 [{shapeType=Box, extentX=8.7199, extentY=12.644, extentZ=4.2843, center=<4.3572 0 -0.0001>}]
 dyn_suspender_l_02 [{shapeType=Box, extentX=8.7199, extentY=15.7313, extentZ=4.2843, center=<4.3572 0 -0.0001>}]
 dyn_suspender_l_03 [{shapeType=Box, extentX=8.7199, extentY=24.8854, extentZ=4.2843, center=<4.3572 0 -0.0001>}]
 dyn_suspender_l_04 [{shapeType=Box, extentX=8.7199, extentY=22.183, extentZ=4.2843, center=<4.3572 0 -0.0001>}]
 dyn_suspender_l_base [{shapeType=Box, extentX=0.5, extentY=0.5, extentZ=0.5}]
 dyn_suspender_l_parent [{shapeType=Box, extentX=10.0871, extentY=16.9638, extentZ=3.3603, center=<6.1278 0 0>}]
 dyn_suspender_r_01 [{shapeType=Box, extentX=8.51, extentY=14.5008, extentZ=3.51, center=<4.1697 0 0>}]
 dyn_suspender_r_02 [{shapeType=Box, extentX=9.6312, extentY=14.7459, extentZ=3.5955, center=<4.4528 -0.0001 0>}]
 dyn_suspender_r_03 [{shapeType=Box, extentX=8.7199, extentY=23.7064, extentZ=4.5343, center=<4.3572 0 -0.0001>}]
 dyn_suspender_r_04 [{shapeType=Box, extentX=8.7199, extentY=22.4973, extentZ=4.8789, center=<4.3572 0 -0.0001>}]
 dyn_suspender_r_base [{shapeType=Box, extentX=1, extentY=1, extentZ=1}]
 dyn_suspender_r_parent [{shapeType=Box, extentX=10.0871, extentY=15.2138, extentZ=3.3603, center=<5.198 0 0>}]
 foot_l [{shapeType=Capsule, length=11.375, radius=8.625, center=<-2.6652 6.2473 1.3792>, rotation=<rot 0 0 -96.2606>}]
 foot_r [{shapeType=Capsule, length=11.375, radius=8.625, center=<2.8556 -5.8945 0.0473>, rotation=<rot 0 0 -96.2606>}]
 pelvis [
  {shapeType=Capsule, length=7.5341, radius=10.5909, center=<-9.1671 -1.106 -0.1962>}
  {shapeType=Capsule, length=9.375, radius=8.75, center=<5.4421 -1.1061 -0.1962>}
 ]
 thigh_l [{shapeType=Capsule, length=20.9619, radius=10.5232, center=<-26.1742 1.4542 5.5797>, rotation=<rot 80.1319 -180 -180>}]
 thigh_r [{shapeType=Capsule, length=22.6872, radius=8.7979, center=<25.8929 -1.437 -5.2696>, rotation=<rot 75.4243 -180 -180>}]
}
constraints [
 {bone1Name=dyn_suspender_r_parent, bone2Name=dyn_suspender_r_base, swing1LimitDegrees=3, swing2LimitDegrees=3, twistLimitDegrees=0.5}
 {bone1Name=dyn_suspender_r_01, bone2Name=dyn_suspender_r_parent, swing1LimitDegrees=10, swing2LimitDegrees=20, twistLimitDegrees=3}
 {bone1Name=dyn_suspender_r_02, bone2Name=dyn_suspender_r_01, swing1LimitDegrees=10, swing2LimitDegrees=20, twistLimitDegrees=3}
 {bone1Name=dyn_suspender_r_03, bone2Name=dyn_suspender_r_02, swing1LimitDegrees=20, swing2LimitDegrees=30, twistLimitDegrees=3}
 {bone1Name=dyn_suspender_r_04, bone2Name=dyn_suspender_r_03, swing1LimitDegrees=20, swing2LimitDegrees=30, twistLimitDegrees=3}
 {bone1Name=dyn_suspender_l_parent, bone2Name=dyn_suspender_l_base, swing1LimitDegrees=10, swing2LimitDegrees=10, twistLimitDegrees=2}
 {bone1Name=dyn_suspender_l_01, bone2Name=dyn_suspender_l_parent, swing1LimitDegrees=10, swing2LimitDegrees=20, twistLimitDegrees=3}
 {bone1Name=dyn_suspender_l_02, bone2Name=dyn_suspender_l_01, swing1LimitDegrees=10, swing2LimitDegrees=20, twistLimitDegrees=3}
 {bone1Name=dyn_suspender_l_03, bone2Name=dyn_suspender_l_02, swing1LimitDegrees=10, swing2LimitDegrees=20, twistLimitDegrees=3}
 {bone1Name=dyn_suspender_l_04, bone2Name=dyn_suspender_l_03, swing1LimitDegrees=10, swing2LimitDegrees=20, twistLimitDegrees=3}
]
