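# Post-process-style AnimBP for the Space Wanderer female body, copying the attached parent's pose and adjusting its suspenders and knee volume.
# Inherited crouch, jump, skydive and pickaxe alphas drive interpolated additive bone offsets on the suspender bases, feeding a RigidBody simulation weighted by SuspenderAlpha.
# IsSitting then enables further corrections.
# Parts-started notifies set it for the SpaceWandererBPIdle context; parts-stopped notifies clear it only for other contexts.
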AnimBlueprint /BRCosmetics/Characters/Player/Female/Medium/Bodies/F_MED_Space_Wanderer/Meshes/F_MED_Space_Wanderer_AnimBP
target_skeleton=&/F_MED_Space_Wanderer_Skeleton
graphs [
 AnimGraph {nodes [
   n1 = Get CrouchMoveAlpha() -> (CrouchMoveAlpha:float)
   n2 = Get JumpAlpha() -> (JumpAlpha:float)
   n3 = "Transform (Modify) Bone - Bone: dyn_suspender_l_base"(
    ComponentPose=Local To Component(LocalPose=Copy Pose From Mesh() {bUseAttachedParent=true})
    Alpha=@n2
    Translation="0,0,0"
    Rotation="-5,30,0"
    Scale="1,1,1"
   ) {
    rotationMode=BMM_Additive
    rotationSpace=BCS_BoneSpace
    alphaBoolBlend {blendOption=Linear}
    alphaScaleBiasClamp {bInterpResult=true, interpSpeedDecreasing=2, interpSpeedIncreasing=6}
    boneToModify {boneName=dyn_suspender_l_base}
    rotation {pitch=-5, yaw=30}
   }
   n4 = "Transform (Modify) Bone - Bone: dyn_suspender_r_base"(ComponentPose=@n3, Alpha=@n2, Translation="0,0,0", Rotation="-30,55,0", Scale="1,1,1") {
    rotationMode=BMM_Additive
    rotationSpace=BCS_BoneSpace
    alphaBoolBlend {blendOption=Linear}
    alphaScaleBiasClamp {bInterpResult=true, interpSpeedDecreasing=2, interpSpeedIncreasing=6}
    boneToModify {boneName=dyn_suspender_r_base}
    rotation {pitch=-30, yaw=55}
   }
   n5 = Get SkydiveAlpha() -> (SkydiveAlpha:float)
   n6 = "Transform (Modify) Bone - Bone: dyn_suspender_l_base"(ComponentPose=@n4, Alpha=@n5, Translation="0,0,0", Rotation="0,20,10", Scale="1,1,1") {
    rotationMode=BMM_Additive
    rotationSpace=BCS_BoneSpace
    alphaBoolBlend {blendOption=Linear}
    alphaScaleBiasClamp {bInterpResult=true, interpSpeedDecreasing=4, interpSpeedIncreasing=6}
    boneToModify {boneName=dyn_suspender_l_base}
    rotation {roll=10, yaw=20}
   }
   n7 = "Transform (Modify) Bone - Bone: dyn_suspender_r_base"(ComponentPose=@n6, Alpha=@n5, Translation="0,0,-5", Rotation="0,-50,-20", Scale="1,1,1") {
    rotationMode=BMM_Additive
    rotationSpace=BCS_BoneSpace
    translationMode=BMM_Additive
    translationSpace=BCS_BoneSpace
    alphaBoolBlend {blendOption=Linear}
    alphaScaleBiasClamp {bInterpResult=true, interpSpeedDecreasing=4, interpSpeedIncreasing=6}
    boneToModify {boneName=dyn_suspender_r_base}
    rotation {roll=-20, yaw=-50}
    translation {z=-5}
   }
   n8 = Get CrouchAlpha() -> (CrouchAlpha:float)
   n9 = "Transform (Modify) Bone - Bone: dyn_suspender_l_base"(ComponentPose=@n7, Alpha=@n8, Translation="0,0,0", Rotation="-35,-70,0", Scale="1,1,1") {
    rotationMode=BMM_Additive
    rotationSpace=BCS_BoneSpace
    alphaBoolBlend {blendOption=Linear}
    alphaScaleBiasClamp {bInterpResult=true, interpSpeedDecreasing=4, interpSpeedIncreasing=60}
    boneToModify {boneName=dyn_suspender_l_base}
    rotation {pitch=-35, yaw=-70}
   }
   n10 = "Transform (Modify) Bone - Bone: dyn_suspender_r_base"(ComponentPose=@n9, Alpha=@n8, Translation="0,0,0", Rotation="-10,40,0", Scale="1,1,1") {
    rotationMode=BMM_Additive
    rotationSpace=BCS_BoneSpace
    alphaBoolBlend {blendOption=Linear}
    alphaScaleBiasClamp {bInterpResult=true, interpSpeedDecreasing=4, interpSpeedIncreasing=4}
    boneToModify {boneName=dyn_suspender_r_base}
    rotation {pitch=-10, yaw=40}
   }
   n11 = "Transform (Modify) Bone - Bone: dyn_suspender_r_base"(ComponentPose=@n10, Alpha=@n1, Translation="0,0,0", Rotation="-25,55,0", Scale="1,1,1") {
    rotationMode=BMM_Additive
    rotationSpace=BCS_BoneSpace
    alphaBoolBlend {blendOption=Linear}
    alphaScaleBiasClamp {bInterpResult=true, interpSpeedDecreasing=4, interpSpeedIncreasing=4}
    boneToModify {boneName=dyn_suspender_r_base}
    rotation {pitch=-25, yaw=55}
   }
   n12 = "Transform (Modify) Bone - Bone: dyn_suspender_l_base"(ComponentPose=@n11, Alpha=@n1, Translation="0,0,0", Rotation="0,-30,0", Scale="1,1,1") {
    rotationMode=BMM_Additive
    rotationSpace=BCS_BoneSpace
    alphaBoolBlend {blendOption=Linear}
    alphaScaleBiasClamp {bInterpResult=true, interpSpeedDecreasing=4, interpSpeedIncreasing=6}
    boneToModify {boneName=dyn_suspender_l_base}
    rotation {yaw=-30}
   }
   n13 = "Transform (Modify) Bone - Bone: dyn_suspender_r_base"(ComponentPose=@n12, Alpha=Get PickaxeSwingMoving() -> (PickaxeSwingMoving:float), Translation="0,0,0", Rotation="-30,25,0", Scale="1,1,1") {
    rotationMode=BMM_Additive
    rotationSpace=BCS_BoneSpace
    alphaBoolBlend {blendOption=Linear}
    alphaScaleBiasClamp {bInterpResult=true, interpSpeedDecreasing=4, interpSpeedIncreasing=4}
    boneToModify {boneName=dyn_suspender_r_base}
    rotation {pitch=-30, yaw=25}
   }
   n14 = RigidBody(ComponentPose=@n13, Alpha=Get SuspenderAlpha() -> (SuspenderAlpha:float)) {
    bClampLinearTranslationLimitToRefPose=true
    bOverrideWorldGravity=true
    overridePhysicsAsset=&/F_MED_Space_Wanderer_Physics
    simulationSpace=BaseBoneSpace
    alphaBoolBlend {blendOption=Linear}
    alphaScaleBiasClamp {interpSpeedDecreasing=5, interpSpeedIncreasing=6}
    baseBoneRef {boneName=root}
    componentAppliedLinearAccClamp=<6000 6000 6000>
    componentLinearAccScale=<0.25 0.25 0.25>
    componentLinearVelScale=<0.25 0.25 0.25>
    overrideWorldGravity {z=-980}
   }
   n15 = Get IsSitting() -> (IsSitting:bool)
   n16 = "Transform (Modify) Bone - Bone: dyn_knee_volume"!(ComponentPose=@n14, bAlphaBoolEnabled:bool=@n15, Translation="-10,0,10", Rotation="0,0,0", Scale="5,5,5") -> (Pose) {
    alphaInputType=Bool
    rotationMode=BMM_Additive
    rotationSpace=BCS_BoneSpace
    scaleMode=BMM_Additive
    scaleSpace=BCS_BoneSpace
    translationMode=BMM_Additive
    translationSpace=BCS_BoneSpace
    alphaBoolBlend {blendOption=Linear}
    boneToModify {boneName=dyn_knee_volume}
    scale=<5 5 5>
    translation {x=-10, z=10}
   }
   n17 = "Transform (Modify) Bone - Bone: dyn_suspender_r_base"!(ComponentPose=@n16, bAlphaBoolEnabled:bool=@n15, Translation="0,0,0", Rotation="-30,20,0", Scale="1,1,1") -> (Pose) {
    alphaInputType=Bool
    rotationMode=BMM_Additive
    rotationSpace=BCS_BoneSpace
    scaleSpace=BCS_BoneSpace
    translationMode=BMM_Additive
    translationSpace=BCS_BoneSpace
    alphaBoolBlend {blendOption=Linear}
    boneToModify {boneName=dyn_suspender_r_base}
    rotation {pitch=-30, yaw=20}
   }
   n18 = "Transform (Modify) Bone - Bone: dyn_suspender_r_base"(ComponentPose=@n17, bAlphaBoolEnabled=@n15, Translation="0,0,0", Rotation="-15,0,0", Scale="1,1,1") {
    alphaInputType=Bool
    rotationMode=BMM_Additive
    rotationSpace=BCS_BoneSpace
    scaleSpace=BCS_BoneSpace
    translationMode=BMM_Additive
    translationSpace=BCS_BoneSpace
    alphaBoolBlend {blendOption=Linear}
    boneToModify {boneName=dyn_suspender_r_base}
    rotation {pitch=-15}
   }
   Output Pose(Result=Component To Local(ComponentPose=@n18)) {layerGroup=None}
  ]}
 EventGraph {nodes [
   n1 = Event OnNotifyPartsStarted!() -> (OutputDelegate, then, Context:Name=None) {is=title K2Node_Event}
   > Branch(Condition="Equal (Name)"!(A:Name=@n1.Context, B:Name=SpaceWandererBPIdle) -> (ReturnValue:bool) {is=title K2Node_CallFunction})
   > Set IsSitting(IsSitting:bool=true) -> (Output_Get:bool)
   n2 = Event OnNotifyPartsStopped!() -> (OutputDelegate, then, Context:Name=None) {is=title K2Node_Event}
   > n3 = Branch(Condition="Equal (Name)"(A=@n2.Context, B=SpaceWandererBPIdle))
   Set IsSitting(execute=@n3.else, IsSitting=false)
  ]}
]
variables [IsSitting {type=boolean}]
