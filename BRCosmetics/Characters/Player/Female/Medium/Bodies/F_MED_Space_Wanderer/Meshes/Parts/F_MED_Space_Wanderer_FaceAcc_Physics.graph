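# Physics asset for the face accessory part of the Space Wanderer female medium body, covering four hair bun bones and the head.
# Each bun (top and bottom, left and right) has a capsule body, alongside a capsule on the head.
# Every bun is constrained directly to the head with identical tight swing and twist limits, allowing only slight wobble rather than free swinging.
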PhysicsAsset /BRCosmetics/Characters/Player/Female/Medium/Bodies/F_MED_Space_Wanderer/Meshes/Parts/F_MED_Space_Wanderer_FaceAcc_Physics
bodies {
 dyn_bun_bottom_l [{shapeType=Capsule, length=3.9413, radius=3.5922, center=<2.9799 0 0>}]
 dyn_bun_bottom_r [{shapeType=Capsule, length=3.9337, radius=3.5864, center=<2.6901 0 -0.0004>}]
 dyn_bun_top_l [{shapeType=Capsule, length=3.8736, radius=3.6195, center=<2.1104 0.0001 0>}]
 dyn_bun_top_r [{shapeType=Capsule, length=3.9574, radius=3.4978, center=<2.6467 -0.014 0>}]
 head [{shapeType=Capsule, length=2.7066, radius=4.0097, rotation=<rot 0 -90 -90>}]
}
constraints [
 {bone1Name=dyn_bun_top_l, bone2Name=head, swing1LimitDegrees=10, swing2LimitDegrees=10, twistLimitDegrees=3}
 {bone1Name=dyn_bun_top_r, bone2Name=head, swing1LimitDegrees=10, swing2LimitDegrees=10, twistLimitDegrees=3}
 {bone1Name=dyn_bun_bottom_l, bone2Name=head, swing1LimitDegrees=10, swing2LimitDegrees=10, twistLimitDegrees=3}
 {bone1Name=dyn_bun_bottom_r, bone2Name=head, swing1LimitDegrees=10, swing2LimitDegrees=10, twistLimitDegrees=3}
]
