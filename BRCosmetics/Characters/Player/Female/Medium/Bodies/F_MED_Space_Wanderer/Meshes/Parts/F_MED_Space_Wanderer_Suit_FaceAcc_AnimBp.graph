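# Animation Blueprint for the face accessory part of the female Space Wanderer suit, posing it from the character mesh it is attached to.
# The copied parent pose is cached and feeds all inputs of a per-bone layered blend filtered at neck_01 and faceAttach.
# On customization load, the Stage2 cosmetic variant tag sets hair_base_active, which the AnimGraph does not read.
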AnimBlueprint /BRCosmetics/Characters/Player/Female/Medium/Bodies/F_MED_Space_Wanderer/Meshes/Parts/F_MED_Space_Wanderer_Suit_FaceAcc_AnimBp
target_skeleton=/Game/Characters/Player/Female/Medium/Bodies/F_MED_Space_Wanderer/Meshes/Parts/F_MED_Space_Wanderer_Suit_FaceAcc_Skeleton
graphs [
 AnimGraph {nodes [
   n1 = Layered blend per bone!(
    BasePose=Use cached pose 'SavedPose10922'()
    BlendPoses_0=Use cached pose 'SavedPose10922'()
    BlendPoses_1:Pose Link Structure=Use cached pose 'SavedPose10922'()
    BlendWeights_0=0.7
    BlendWeights_1:float=1
   ) -> (Pose) {
    blendPoses [{linkId=-1, sourceLinkId=-1}, {linkId=-1, sourceLinkId=-1}]
    blendWeights [1,1]
    layerSetup [{branchFilters [{blendDepth=1, boneName=neck_01}]}, {branchFilters [{blendDepth=1, boneName=faceAttach}]}]
   }
   Output Pose(Result=@n1) {layerGroup=None}
   Save cached pose 'SavedPose10922'(Pose=Copy Pose From Mesh() {bUseAttachedParent=true}) {cacheName=SavedPose10922}
  ]}
 EventGraph {nodes [
   Event OnCharacterCustomizationLoaded()
   > n1 = Branch(Condition=CharacterItemDefinitionHasAnyMetaTag(Tags=Make Array("[0]"=(TagName="Cosmetics.Variant.Property.Stage2"))))
   > Set hair_base_active(hair_base_active:bool=true) -> (Output_Get:bool)
   Set hair_base_active(execute=@n1.else, hair_base_active=false)
   Event BlueprintInitializeAnimation()
   > Set MeshToCopy(MeshToCopy:Skeletal Mesh Component Object Reference=GetOwnerSkeletalMeshComponent()) -> (Output_Get:Skeletal Mesh Component Object Reference)
   TryGetPawnOwner()
   Event BlueprintUpdateAnimation()
  ]}
]
variables [
 hair_base_active {type=boolean}
 MeshToCopy {type=SkeletalMeshComponent}
]
